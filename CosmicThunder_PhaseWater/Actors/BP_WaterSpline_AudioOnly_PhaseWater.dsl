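; BP_WaterSpline_AudioOnly_PhaseWater is an audio-only water spline actor that switches which spline water audio components play for phase water.
; On BeginPlay, if the GetWaterBodyEnabled scalable float is true, it waits asynchronously for playlist data to be ready.
; It then activates each FortSplineWaterAudioComponent tagged "SunRoseWater" and deactivates all others.
; Overlap and Tick events are empty, and construction only calls the parent.
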(event Collision|EventActorBeginOverlap
  (params (OtherActor :type Actor%20Object%20Reference)))

(event EventTick
  (params (DeltaSeconds :type Float%20%28single-precision%29)))

(event EventBeginPlay
  (|CallParentFunction)
  (if (ScalableFloat|GetValueasBool (Variables|Default|GetWaterBodyEnabled))
    (Class|FortAsyncActionCurrentPlaylistReady|PlaylistDataReadyAsync
      (:OnReady
        (for _array_element (Actor|GetComponentsByClass "/Script/FortniteGame.FortSplineWaterAudioComponent")
          (if (Utilities|Array|ContainsItem (Variables|/Script/Engine.ActorComponent|GetComponentTags _array_element) "SunRoseWater")
            (Components|Activation|SetActive _array_element true true)
            (else
              (Components|Activation|SetActive _array_element false true))))))))

(fn ConstructionScript
  (|CallParentFunction))
